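# Physics asset for the AvocadoSeal two-handed melee pickaxe mesh, giving it a simple collision shape.
# It has a single body on the Melee_TwoHanded_Handle bone, made of one capsule offset along and roughly aligned with the handle.
# There are no constraints, so the whole axe acts as one rigid body with no articulated parts.
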PhysicsAsset /BRCosmetics/Weapons/FORT_Melee/Pickaxe_AvocadoSeal/Meshes/AvocadoSeal_Axe_PhysicsAsset
bodies {Melee_TwoHanded_Handle [{shapeType=Capsule, length=39.8237, radius=19.159, center=<-0.0093 -6.2357 18.4255>, rotation=<rot -2.482 -90 -0.0129>}]}
constraints []
